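# MetaSound Source for the looping campfire fire ambience, on the Athena ambient sound class with trap attenuation.
# On play, it randomly picks one of two loop wave variations and feeds it to a stereo Wave Player.
# An ADSR envelope, triggered as playback starts, multiplies the player's left channel to fade the sound in and sends it to the mono output.
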MetaSoundSource /FNGameplayCues/Athena/Sounds/Campfire/EnvCampfire_Fire_Loop_MSS
inputs [
 LoopVariations {type=WaveAsset:Array}
 UE.Source.OnPlay {type=Trigger}
]
interfaces [UE.OutputFormat.Mono:1.0, UE.Source:1.1]
outputs ["UE.OutputFormat.Mono.Audio:0" {type=Audio}]
pages [Default {
  input_defaults {UE.Source.OnPlay=null, LoopVariations [/Game/Athena/Sounds/Campfire/sfx_campfire_loop_01, /Game/Athena/Sounds/Campfire/sfx_campfire_loop_02]}
  nodes [
   ArrayRandomGet_0 = Array.Random Get:WaveAsset:Array(In Array=^LoopVariations, Next=^UE.Source.OnPlay) {also {version=~}}
   WavePlayer_0 = UE.Wave Player:Stereo(Play=@ArrayRandomGet_0.On Next, Wave Asset=@ArrayRandomGet_0.Value) {also {defaults {loop="<MetaSoundBuilderResult.SUCCEEDED: 0>"}}}
   ADSREnvelopeADSREnvelope_0 = ADSR Envelope.ADSR Envelope:Audio(Trigger Attack=@WavePlayer_0.On Play) {also {defaults {Attack Time=1, Sustain Level=1}}}
   Multiply_0 = UE.Multiply:Audio(PrimaryOperand=@ADSREnvelopeADSREnvelope_0.Out Envelope, AdditionalOperands=@WavePlayer_0.Out Left)
  ]
  links [{dst="@UE.OutputFormat.Mono.Audio:0", src=Multiply_0.Out}]
 }]
properties {
 assetImportData=&self:AssetImportData
 attenuationSettings=/Game/Sounds/Attenuation/Fort_Traps_Athena_Att
 bLooping=true
 duration=10000
 maxDistance=1200
 soundClassObject=&Fort_Ambient_Athena
 volume=4.158
}
