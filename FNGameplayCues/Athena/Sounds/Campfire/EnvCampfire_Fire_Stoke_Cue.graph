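# Sound cue for stoking a campfire in Athena, playing a fire flare-up sound.
# A single wave player feeds a Modulator at its default settings, which slightly varies pitch and volume on each play.
# The cue uses the shared Athena traps attenuation and the Athena ambient sound class, with a boosted volume multiplier.
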SoundCue /FNGameplayCues/Athena/Sounds/Campfire/EnvCampfire_Fire_Stoke_Cue
root=Modulator_0
nodes [
 {id=Modulator_0, children [WavePlayer_1]}
 {id=WavePlayer_1, properties {soundWaveAssetPtr=&/Fire_Campfire_Stoke_Flare_Up_01}}
]
properties {
 attenuationSettings=/Game/Sounds/Attenuation/Fort_Traps_Athena_Att
 maxDistance=1200
 soundClassObject=&Fort_Ambient_Athena
 volumeMultiplier=1.455
}
