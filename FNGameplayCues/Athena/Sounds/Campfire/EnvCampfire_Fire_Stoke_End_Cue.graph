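# Sound cue for the campfire stoke ending, playing a single die-out fire sound in the Athena gameplay cues.
# One wave player feeds a modulator that slightly randomizes pitch on each play.
# It borrows a medium-object demolition attenuation and the impacts sound class, with object-impact concurrency referenced alongside a prevent-new override setting.
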SoundCue /FNGameplayCues/Athena/Sounds/Campfire/EnvCampfire_Fire_Stoke_End_Cue
root=Modulator_5
nodes [
 {id=Modulator_5, children [WavePlayer_0], properties {pitchMax=1.1, pitchMin=0.9}}
 {id=WavePlayer_0, properties {soundWaveAssetPtr=&/Fire_Campfire_Stoke_Die_Out_01}}
]
properties {
 attenuationSettings=/Game/Sounds/Attenuation/Fort_Demolition_Destroy_Medium_Sized_Object_Att
 maxDistance=4000
 soundClassObject=&Fort_Impacts
 volumeMultiplier=0.56
 concurrencyOverrides {resolutionRule=PreventNew}
 concurrencySet [&CONC_Object_Impacts_FourMax]
}
