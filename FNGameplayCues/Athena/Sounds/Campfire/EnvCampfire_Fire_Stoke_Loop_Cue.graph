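# Sound cue for the looping campfire stoke fire ambience in Athena, named for a stoked, hot-burning campfire.
# A single looping wave player feeds the hot burn loop wave straight to the output, with no randomization or modulation.
# It uses the Athena ambient sound class and borrows the Athena traps attenuation asset for distance falloff.
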SoundCue /FNGameplayCues/Athena/Sounds/Campfire/EnvCampfire_Fire_Stoke_Loop_Cue
root=WavePlayer_3
nodes [{id=WavePlayer_3, properties {bLooping=true, soundWaveAssetPtr=&/Fire_Campfire_Stoke_Burn_Hot_Loop_01}}]
properties {
 attenuationSettings=/Game/Sounds/Attenuation/Fort_Traps_Athena_Att
 maxDistance=1200
 soundClassObject=&Fort_Ambient_Athena
 volumeMultiplier=1.039
}
